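# Animation Blueprint for the winged Cavalry Alt backpack cosmetic that poses the wings to match player movement.
# A state machine selects base, targeting, jump, fall, crouch or nested skydive poses from a PoseAsset, mostly with acceleration-driven additive lean blendspaces.
# Chained AnimDynamics spring bones add wing secondary motion before the pose is cached.
# Speed-weighted run and sprint additives layer on top, bypassed when crouching or skydiving, before a small root offset.
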AnimBlueprint /BRCosmetics/Accessories/FORT_Backpacks/Backpack_F_MED_Cavalry_Alt/Meshes/F_MED_Cavalry_Alt_Pack_AnimBP
target_skeleton=/Game/Accessories/FORT_Backpacks/Backpack_F_MED_Cavalry_Alt/Meshes/F_MED_Cavalry_Alt_Pack_Skeleton
graphs [
 AnimGraph {nodes [
   n1 = Wings_Movement_Poses_PoseAsset_Add : Run!(PoseWeight:float=1) -> (Pose:Pose Link Structure) {
    is=title AnimGraphNode_PoseByName
    poseAsset=/BRCosmetics/Accessories/FORT_Backpacks/Backpack_F_MED_Cavalry_Alt/Animation/Wings_Movement_Poses_PoseAsset_Add
    poseName=Run
   }
   n2 = Wings_Movement_Poses_PoseAsset_Add : Sprint!(PoseWeight:float=1) -> (Pose:Pose Link Structure) {
    is=title AnimGraphNode_PoseByName
    poseAsset=&Wings_Movement_Poses_PoseAsset_Add
    poseName=Sprint
   }
   n3 = Get SprintSpeedRemap() -> (SprintSpeedRemap:float)
   n4 = Two Way Blend(
    A=Apply Additive(Base=Use cached pose 'SavedPose21935'(), Additive=@n2, Alpha=@n3) {alphaBoolBlend {blendOption=Linear}, alphaScaleBiasClamp {bInterpResult=true, interpSpeedDecreasing=5}}
    B=Apply Additive(Base=Use cached pose 'SavedPose21935'(), Additive=@n1, Alpha=Get RunSpeedRemap() -> (RunSpeedRemap:float)) {alphaBoolBlend {blendOption=Linear}, alphaScaleBiasClamp {bInterpResult=true, interpSpeedDecreasing=5}}
    Alpha=@n3
   ) {blendNode {alphaBoolBlend {blendOption=Linear}, alphaScaleBiasClamp {bInterpResult=true, bias=1, interpSpeedDecreasing=5, scale=-1}}}
   n5 = Blend Poses by bool(BlendPose_0=Use cached pose 'SavedPose21935'(), BlendPose_1=@n4, BlendTime_0=0.3, BlendTime_1=0.3, bActiveValue=Get bIsCrouching()) {blendType=Linear}
   n6 = Use cached pose 'SavedPose21935'()
   n7 = Get bIsSkyDiving()
   n8 = "Transform (Modify) Bone - Bone: root"(
    ComponentPose=Local To Component(LocalPose=Blend Poses by bool(BlendPose_0=@n6, BlendPose_1=@n5, BlendTime_0=0.3, BlendTime_1=0.3, bActiveValue=@n7) {blendType=Linear})
    Translation="0,1,0"
    Rotation="0,0,0"
   ) {
    translationMode=BMM_Additive
    translationSpace=BCS_BoneSpace
    alphaBoolBlend {blendOption=Linear}
    boneToModify {boneName=root}
    translation {y=1}
   }
   Output Pose(Result=Component To Local(ComponentPose=@n8)) {layerGroup=None}
   AnimGraphNode_StateMachine_0 = Wings_position() {is=title AnimGraphNode_StateMachine}
   n9 = Get DynamicsAlpha() -> (DynamicsAlpha:float)
   n10 = Anim Dynamics - Bone: dyn_wing_index_r(ComponentPose=Local To Component(LocalPose=@AnimGraphNode_StateMachine_0), Alpha=@n9, ExternalForce="0,0,0") {
    angularDampingOverride=0.9999
    angularSpringConstant=8000
    bAngularSpring=true
    bGravityOverrideInSimSpace=true
    bOverrideAngularDamping=true
    bUseGravityOverride=true
    simulationSpace=RootRelative
    alphaBoolBlend {blendOption=Linear}
    boundBone {boneName=dyn_wing_index_r}
    componentAppliedLinearAccClamp=<2000 2000 2000>
    componentLinearAccScale=<5 5 5>
    componentLinearVelScale=<5 5 5>
    gravityOverride {z=-980}
    physicsBodyDefinitions [{
      boundBone {boneName=dyn_wing_index_r}
      boxExtents=<10 20 20>
      constraintSetup {angularLimitsMax=<0 5 2>, angularLimitsMin=<0 -5 -2>, angularTarget=<0 0 0>}
      localJointOffset=<0.3 0 0>
     }]
   }
   n11 = Anim Dynamics - Bone: dyn_wing_index_02_r(ComponentPose=@n10, Alpha=@n9, ExternalForce="0,0,0") {
    angularDampingOverride=0.9999
    angularSpringConstant=8000
    bAngularSpring=true
    bGravityOverrideInSimSpace=true
    bOverrideAngularDamping=true
    bUseGravityOverride=true
    simulationSpace=RootRelative
    alphaBoolBlend {blendOption=Linear}
    boundBone {boneName=dyn_wing_index_02_r}
    componentAppliedLinearAccClamp=<2000 2000 2000>
    componentLinearAccScale=<5 5 5>
    componentLinearVelScale=<5 5 5>
    gravityOverride {z=-980}
    physicsBodyDefinitions [{
      boundBone {boneName=dyn_wing_index_02_r}
      boxExtents=<10 20 20>
      constraintSetup {angularLimitsMax=<0 5 2>, angularLimitsMin=<0 -5 -2>, angularTarget=<0 0 0>}
      localJointOffset=<0.3 0 0>
     }]
   }
   n12 = Anim Dynamics - Bone: dyn_wing_index_03_r(ComponentPose=@n11, Alpha=@n9, ExternalForce="0,0,0") {
    angularDampingOverride=0.9999
    angularSpringConstant=8000
    bAngularSpring=true
    bGravityOverrideInSimSpace=true
    bOverrideAngularDamping=true
    bUseGravityOverride=true
    simulationSpace=RootRelative
    alphaBoolBlend {blendOption=Linear}
    boundBone {boneName=dyn_wing_index_03_r}
    componentAppliedLinearAccClamp=<2000 2000 2000>
    componentLinearAccScale=<5 5 5>
    componentLinearVelScale=<5 5 5>
    gravityOverride {z=-980}
    physicsBodyDefinitions [{
      boundBone {boneName=dyn_wing_index_03_r}
      boxExtents=<10 20 20>
      constraintSetup {angularLimitsMax=<0 5 2>, angularLimitsMin=<0 -5 -2>, angularTarget=<0 0 0>}
      localJointOffset=<0.3 0 0>
     }]
   }
   n13 = Anim Dynamics - Bone: dyn_wing_mid_r(ComponentPose=@n12, Alpha=@n9, ExternalForce="0,0,0") {
    angularDampingOverride=0.9999
    angularSpringConstant=8000
    bAngularSpring=true
    bGravityOverrideInSimSpace=true
    bOverrideAngularDamping=true
    bUseGravityOverride=true
    simulationSpace=RootRelative
    alphaBoolBlend {blendOption=Linear}
    boundBone {boneName=dyn_wing_mid_r}
    componentAppliedLinearAccClamp=<2000 2000 2000>
    componentLinearAccScale=<5 5 5>
    componentLinearVelScale=<5 5 5>
    gravityOverride {z=-980}
    physicsBodyDefinitions [{
      boundBone {boneName=dyn_wing_mid_r}
      boxExtents=<10 20 20>
      constraintSetup {angularLimitsMax=<0 5 2>, angularLimitsMin=<0 -5 -2>, angularTarget=<0 0 0>}
      localJointOffset=<0.3 0 0>
     }]
   }
   n14 = Anim Dynamics - Bone: dyn_wing_mid_02_r(ComponentPose=@n13, Alpha=@n9, ExternalForce="0,0,0") {
    angularDampingOverride=0.9999
    angularSpringConstant=8000
    bAngularSpring=true
    bGravityOverrideInSimSpace=true
    bOverrideAngularDamping=true
    bUseGravityOverride=true
    simulationSpace=RootRelative
    alphaBoolBlend {blendOption=Linear}
    boundBone {boneName=dyn_wing_mid_02_r}
    componentAppliedLinearAccClamp=<2000 2000 2000>
    componentLinearAccScale=<5 5 5>
    componentLinearVelScale=<5 5 5>
    gravityOverride {z=-980}
    physicsBodyDefinitions [{
      boundBone {boneName=dyn_wing_mid_02_r}
      boxExtents=<10 20 20>
      constraintSetup {angularLimitsMax=<0 5 2>, angularLimitsMin=<0 -5 -2>, angularTarget=<0 0 0>}
      localJointOffset=<0.3 0 0>
     }]
   }
   n15 = Anim Dynamics - Bone: dyn_wing_mid_03_r(ComponentPose=@n14, Alpha=@n9, ExternalForce="0,0,0") {
    angularDampingOverride=0.9999
    angularSpringConstant=8000
    bAngularSpring=true
    bGravityOverrideInSimSpace=true
    bOverrideAngularDamping=true
    bUseGravityOverride=true
    simulationSpace=RootRelative
    alphaBoolBlend {blendOption=Linear}
    boundBone {boneName=dyn_wing_mid_03_r}
    componentAppliedLinearAccClamp=<2000 2000 2000>
    componentLinearAccScale=<5 5 5>
    componentLinearVelScale=<5 5 5>
    gravityOverride {z=-980}
    physicsBodyDefinitions [{
      boundBone {boneName=dyn_wing_mid_03_r}
      boxExtents=<10 20 20>
      constraintSetup {angularLimitsMax=<0 5 2>, angularLimitsMin=<0 -5 -2>, angularTarget=<0 0 0>}
      localJointOffset=<0.3 0 0>
     }]
   }
   n16 = Anim Dynamics - Bone: dyn_wing_pinky_r(ComponentPose=@n15, Alpha=@n9, ExternalForce="0,0,0") {
    angularDampingOverride=0.9999
    angularSpringConstant=8000
    bAngularSpring=true
    bGravityOverrideInSimSpace=true
    bOverrideAngularDamping=true
    bUseGravityOverride=true
    simulationSpace=RootRelative
    alphaBoolBlend {blendOption=Linear}
    boundBone {boneName=dyn_wing_pinky_r}
    componentAppliedLinearAccClamp=<2000 2000 2000>
    componentLinearAccScale=<5 5 5>
    componentLinearVelScale=<5 5 5>
    gravityOverride {z=-980}
    physicsBodyDefinitions [{
      boundBone {boneName=dyn_wing_pinky_r}
      boxExtents=<10 20 20>
      constraintSetup {angularLimitsMax=<0 5 2>, angularLimitsMin=<0 -5 -2>, angularTarget=<0 0 0>}
      localJointOffset=<0.3 0 0>
     }]
   }
   n17 = Anim Dynamics - Bone: dyn_wing_pinky_02_r(ComponentPose=@n16, Alpha=@n9, ExternalForce="0,0,0") {
    angularDampingOverride=0.9999
    angularSpringConstant=8000
    bAngularSpring=true
    bGravityOverrideInSimSpace=true
    bOverrideAngularDamping=true
    bUseGravityOverride=true
    simulationSpace=RootRelative
    alphaBoolBlend {blendOption=Linear}
    boundBone {boneName=dyn_wing_pinky_02_r}
    componentAppliedLinearAccClamp=<2000 2000 2000>
    componentLinearAccScale=<5 5 5>
    componentLinearVelScale=<5 5 5>
    gravityOverride {z=-980}
    physicsBodyDefinitions [{
      boundBone {boneName=dyn_wing_pinky_02_r}
      boxExtents=<10 20 20>
      constraintSetup {angularLimitsMax=<0 5 2>, angularLimitsMin=<0 -5 -2>, angularTarget=<0 0 0>}
      localJointOffset=<0.3 0 0>
     }]
   }
   n18 = Anim Dynamics - Bone: dyn_wing_index_l(ComponentPose=@n17, Alpha=@n9, ExternalForce="0,0,0") {
    angularDampingOverride=0.9999
    angularSpringConstant=8000
    bAngularSpring=true
    bGravityOverrideInSimSpace=true
    bOverrideAngularDamping=true
    bUseGravityOverride=true
    simulationSpace=RootRelative
    alphaBoolBlend {blendOption=Linear}
    boundBone {boneName=dyn_wing_index_l}
    componentAppliedLinearAccClamp=<2000 2000 2000>
    componentLinearAccScale=<5 5 5>
    componentLinearVelScale=<5 5 5>
    gravityOverride {z=-980}
    physicsBodyDefinitions [{
      boundBone {boneName=dyn_wing_index_l}
      boxExtents=<10 20 20>
      constraintSetup {angularLimitsMax=<0 5 2>, angularLimitsMin=<0 -5 -2>, angularTarget=<0 0 0>}
      localJointOffset=<-0.3 0 0>
     }]
   }
   n19 = Anim Dynamics - Bone: dyn_wing_index_02_l(ComponentPose=@n18, Alpha=@n9, ExternalForce="0,0,0") {
    angularDampingOverride=0.9999
    angularSpringConstant=8000
    bAngularSpring=true
    bGravityOverrideInSimSpace=true
    bOverrideAngularDamping=true
    bUseGravityOverride=true
    simulationSpace=RootRelative
    alphaBoolBlend {blendOption=Linear}
    boundBone {boneName=dyn_wing_index_02_l}
    componentAppliedLinearAccClamp=<2000 2000 2000>
    componentLinearAccScale=<5 5 5>
    componentLinearVelScale=<5 5 5>
    gravityOverride {z=-980}
    physicsBodyDefinitions [{
      boundBone {boneName=dyn_wing_index_02_l}
      boxExtents=<10 20 20>
      constraintSetup {angularLimitsMax=<0 5 2>, angularLimitsMin=<0 -5 -2>, angularTarget=<0 0 0>}
      localJointOffset=<-0.3 0 0>
     }]
   }
   n20 = Anim Dynamics - Bone: dyn_wing_index_03_l(ComponentPose=@n19, Alpha=@n9, ExternalForce="0,0,0") {
    angularDampingOverride=0.9999
    angularSpringConstant=8000
    bAngularSpring=true
    bGravityOverrideInSimSpace=true
    bOverrideAngularDamping=true
    bUseGravityOverride=true
    simulationSpace=RootRelative
    alphaBoolBlend {blendOption=Linear}
    boundBone {boneName=dyn_wing_index_03_l}
    componentAppliedLinearAccClamp=<2000 2000 2000>
    componentLinearAccScale=<5 5 5>
    componentLinearVelScale=<5 5 5>
    gravityOverride {z=-980}
    physicsBodyDefinitions [{
      boundBone {boneName=dyn_wing_index_03_l}
      boxExtents=<10 20 20>
      constraintSetup {angularLimitsMax=<0 5 2>, angularLimitsMin=<0 -5 -2>, angularTarget=<0 0 0>}
      localJointOffset=<-0.3 0 0>
     }]
   }
   n21 = Anim Dynamics - Bone: dyn_wing_mid_l(ComponentPose=@n20, Alpha=@n9, ExternalForce="0,0,0") {
    angularDampingOverride=0.9999
    angularSpringConstant=8000
    bAngularSpring=true
    bGravityOverrideInSimSpace=true
    bOverrideAngularDamping=true
    bUseGravityOverride=true
    simulationSpace=RootRelative
    alphaBoolBlend {blendOption=Linear}
    boundBone {boneName=dyn_wing_mid_l}
    componentAppliedLinearAccClamp=<2000 2000 2000>
    componentLinearAccScale=<5 5 5>
    componentLinearVelScale=<5 5 5>
    gravityOverride {z=-980}
    physicsBodyDefinitions [{
      boundBone {boneName=dyn_wing_mid_l}
      boxExtents=<10 20 20>
      constraintSetup {angularLimitsMax=<0 5 2>, angularLimitsMin=<0 -5 -2>, angularTarget=<0 0 0>}
      localJointOffset=<-0.3 0 0>
     }]
   }
   n22 = Anim Dynamics - Bone: dyn_wing_mid_02_l(ComponentPose=@n21, Alpha=@n9, ExternalForce="0,0,0") {
    angularDampingOverride=0.9999
    angularSpringConstant=8000
    bAngularSpring=true
    bGravityOverrideInSimSpace=true
    bOverrideAngularDamping=true
    bUseGravityOverride=true
    simulationSpace=RootRelative
    alphaBoolBlend {blendOption=Linear}
    boundBone {boneName=dyn_wing_mid_02_l}
    componentAppliedLinearAccClamp=<2000 2000 2000>
    componentLinearAccScale=<5 5 5>
    componentLinearVelScale=<5 5 5>
    gravityOverride {z=-980}
    physicsBodyDefinitions [{
      boundBone {boneName=dyn_wing_mid_02_l}
      boxExtents=<10 20 20>
      constraintSetup {angularLimitsMax=<0 5 2>, angularLimitsMin=<0 -5 -2>, angularTarget=<0 0 0>}
      localJointOffset=<-0.3 0 0>
     }]
   }
   n23 = Anim Dynamics - Bone: dyn_wing_mid_03_l(ComponentPose=@n22, Alpha=@n9, ExternalForce="0,0,0") {
    angularDampingOverride=0.9999
    angularSpringConstant=8000
    bAngularSpring=true
    bGravityOverrideInSimSpace=true
    bOverrideAngularDamping=true
    bUseGravityOverride=true
    simulationSpace=RootRelative
    alphaBoolBlend {blendOption=Linear}
    boundBone {boneName=dyn_wing_mid_03_l}
    componentAppliedLinearAccClamp=<2000 2000 2000>
    componentLinearAccScale=<5 5 5>
    componentLinearVelScale=<5 5 5>
    gravityOverride {z=-980}
    physicsBodyDefinitions [{
      boundBone {boneName=dyn_wing_mid_03_l}
      boxExtents=<10 20 20>
      constraintSetup {angularLimitsMax=<0 5 2>, angularLimitsMin=<0 -5 -2>, angularTarget=<0 0 0>}
      localJointOffset=<-0.3 0 0>
     }]
   }
   n24 = Anim Dynamics - Bone: dyn_wing_pinky_l(ComponentPose=@n23, Alpha=@n9, ExternalForce="0,0,0") {
    angularDampingOverride=0.9999
    angularSpringConstant=8000
    bAngularSpring=true
    bGravityOverrideInSimSpace=true
    bOverrideAngularDamping=true
    bUseGravityOverride=true
    simulationSpace=RootRelative
    alphaBoolBlend {blendOption=Linear}
    boundBone {boneName=dyn_wing_pinky_l}
    componentAppliedLinearAccClamp=<2000 2000 2000>
    componentLinearAccScale=<5 5 5>
    componentLinearVelScale=<5 5 5>
    gravityOverride {z=-980}
    physicsBodyDefinitions [{
      boundBone {boneName=dyn_wing_pinky_l}
      boxExtents=<10 20 20>
      constraintSetup {angularLimitsMax=<0 5 2>, angularLimitsMin=<0 -5 -2>, angularTarget=<0 0 0>}
      localJointOffset=<-0.3 0 0>
     }]
   }
   n25 = Anim Dynamics - Bone: dyn_wing_pinky_02_l(ComponentPose=@n24, Alpha=@n9, ExternalForce="0,0,0") {
    angularDampingOverride=0.9999
    angularSpringConstant=8000
    bAngularSpring=true
    bGravityOverrideInSimSpace=true
    bOverrideAngularDamping=true
    bUseGravityOverride=true
    simulationSpace=RootRelative
    alphaBoolBlend {blendOption=Linear}
    boundBone {boneName=dyn_wing_pinky_02_l}
    componentAppliedLinearAccClamp=<2000 2000 2000>
    componentLinearAccScale=<5 5 5>
    componentLinearVelScale=<5 5 5>
    gravityOverride {z=-980}
    physicsBodyDefinitions [{
      boundBone {boneName=dyn_wing_pinky_02_l}
      boxExtents=<10 20 20>
      constraintSetup {angularLimitsMax=<0 5 2>, angularLimitsMin=<0 -5 -2>, angularTarget=<0 0 0>}
      localJointOffset=<-0.3 0 0>
     }]
   }
   Save cached pose 'SavedPose21935'(Pose=Component To Local(ComponentPose=@n25)) {cacheName=SavedPose21935}
  ]}
 Wings_position {
  entry=Base
  path=AnimGraph.AnimGraphNode_StateMachine_0.Wings_position
  ids=id
  nodes [
   AnimStateNode_0 = Base() {is=title AnimStateNode}
   AnimStateNode_1 = Targeting() {is=title AnimStateNode}
   AnimStateNode_2 = Jumping() {is=title AnimStateNode}
   AnimStateNode_3 = Falling() {is=title AnimStateNode}
   AnimStateNode_4 = SkydiveMode() {is=title AnimStateNode}
   AnimStateNode_5 = Crouching() {is=title AnimStateNode}
   AnimStateTransitionNode_15 = AnimStateTransitionNode()
   AnimStateTransitionNode_16 = AnimStateTransitionNode()
   AnimStateTransitionNode_17 = AnimStateTransitionNode()
   AnimStateTransitionNode_18 = AnimStateTransitionNode()
   AnimStateTransitionNode_19 = AnimStateTransitionNode()
   AnimStateTransitionNode_2 = AnimStateTransitionNode()
   AnimStateTransitionNode_20 = AnimStateTransitionNode()
   AnimStateTransitionNode_21 = AnimStateTransitionNode()
   AnimStateTransitionNode_22 = AnimStateTransitionNode()
   AnimStateTransitionNode_23 = AnimStateTransitionNode()
   AnimStateTransitionNode_24 = AnimStateTransitionNode()
   AnimStateTransitionNode_3 = AnimStateTransitionNode()
  ]
  links [
   {
    dst=Crouching
    rule=AnimGraph.AnimGraphNode_StateMachine_0.Wings_position.AnimStateTransitionNode_21.Transition
    src=Base
    via=AnimStateTransitionNode_21
   }
   {
    dst=Falling
    rule=AnimGraph.AnimGraphNode_StateMachine_0.Wings_position.AnimStateTransitionNode_17.Transition
    src=Base
    via=AnimStateTransitionNode_17
   }
   {
    dst=Jumping
    rule=AnimGraph.AnimGraphNode_StateMachine_0.Wings_position.AnimStateTransitionNode_2.Transition
    src=Base
    via=AnimStateTransitionNode_2
   }
   {
    dst=SkydiveMode
    rule=AnimGraph.AnimGraphNode_StateMachine_0.Wings_position.AnimStateTransitionNode_20.Transition
    src=Base
    via=AnimStateTransitionNode_20
   }
   {
    dst=Targeting
    rule=AnimGraph.AnimGraphNode_StateMachine_0.Wings_position.AnimStateTransitionNode_23.Transition
    src=Base
    via=AnimStateTransitionNode_23
   }
   {
    dst=Base
    rule=AnimGraph.AnimGraphNode_StateMachine_0.Wings_position.AnimStateTransitionNode_22.Transition
    src=Crouching
    via=AnimStateTransitionNode_22
   }
   {
    dst=Base
    rule=AnimGraph.AnimGraphNode_StateMachine_0.Wings_position.AnimStateTransitionNode_16.Transition
    src=Falling
    via=AnimStateTransitionNode_16
   }
   {
    dst=Base
    rule=AnimGraph.AnimGraphNode_StateMachine_0.Wings_position.AnimStateTransitionNode_3.Transition
    src=Jumping
    via=AnimStateTransitionNode_3
   }
   {
    dst=Falling
    rule=AnimGraph.AnimGraphNode_StateMachine_0.Wings_position.AnimStateTransitionNode_15.Transition
    src=Jumping
    via=AnimStateTransitionNode_15
   }
   {
    dst=SkydiveMode
    rule=AnimGraph.AnimGraphNode_StateMachine_0.Wings_position.AnimStateTransitionNode_18.Transition
    src=Jumping
    via=AnimStateTransitionNode_18
   }
   {
    dst=Base
    rule=AnimGraph.AnimGraphNode_StateMachine_0.Wings_position.AnimStateTransitionNode_19.Transition
    src=SkydiveMode
    via=AnimStateTransitionNode_19
   }
   {
    dst=Base
    rule=AnimGraph.AnimGraphNode_StateMachine_0.Wings_position.AnimStateTransitionNode_24.Transition
    src=Targeting
    via=AnimStateTransitionNode_24
   }
  ]
 }
 Base {
  path=AnimGraph.AnimGraphNode_StateMachine_0.Wings_position.AnimStateNode_0.Base
  nodes [
   n1 = Get PlayerForwardAcceleration() -> (PlayerForwardAcceleration:float)
   n2 = Get PlayerSidewaysAcceleration() -> (PlayerSidewaysAcceleration:float)
   n3 = Wings_Movement_Poses_PoseAsset : Base!(PoseWeight:float=1) -> (Pose:Pose Link Structure) {
    is=title AnimGraphNode_PoseByName
    poseAsset=/BRCosmetics/Accessories/FORT_Backpacks/Backpack_F_MED_Cavalry_Alt/Animation/Wings_Movement_Poses_PoseAsset
    poseName=Base
   }
   Output Animation Pose(Result=Apply Additive(
     Base=@n3
     Additive=Blendspace Player 'BS_Wings_Leans_Ground'(X=@n2, Y=@n1) {blendSpace=/BRCosmetics/Accessories/FORT_Backpacks/Backpack_F_MED_Cavalry_Alt/Animation/BS_Wings_Leans_Ground}
    ) {alphaBoolBlend {blendOption=Linear}})
  ]
 }
 Targeting {
  path=AnimGraph.AnimGraphNode_StateMachine_0.Wings_position.AnimStateNode_1.Targeting
  nodes [
   n1 = Get PlayerForwardAcceleration()
   n2 = Get PlayerSidewaysAcceleration()
   n3 = Wings_Movement_Poses_PoseAsset : Targeting!(PoseWeight:float=1) -> (Pose:Pose Link Structure) {
    is=title AnimGraphNode_PoseByName
    poseAsset=&Wings_Movement_Poses_PoseAsset
    poseName=Targeting
   }
   Output Animation Pose(Result=Apply Additive(
     Base=@n3
     Additive=Blendspace Player 'BS_Wings_Leans_Ground'(X=@n2, Y=@n1) {blendSpace=&BS_Wings_Leans_Ground}
    ) {alphaBoolBlend {blendOption=Linear}})
  ]
 }
 Jumping {
  path=AnimGraph.AnimGraphNode_StateMachine_0.Wings_position.AnimStateNode_2.Jumping
  nodes [
   n1 = Get PlayerForwardAcceleration()
   n2 = Get PlayerSidewaysAcceleration()
   n3 = Wings_Movement_Poses_PoseAsset : Jump!(PoseWeight:float=1) -> (Pose:Pose Link Structure) {
    is=title AnimGraphNode_PoseByName
    poseAsset=&Wings_Movement_Poses_PoseAsset
    poseName=Jump
   }
   Output Animation Pose(Result=Apply Additive(
     Base=@n3
     Additive=Blendspace Player 'BS_Wings_Leans_Ground'(X=@n2, Y=@n1) {blendSpace=&BS_Wings_Leans_Ground}
    ) {alphaBoolBlend {blendOption=Linear}})
  ]
 }
 Falling {
  path=AnimGraph.AnimGraphNode_StateMachine_0.Wings_position.AnimStateNode_3.Falling
  nodes [
   n1 = Get PlayerForwardAcceleration()
   n2 = Get PlayerSidewaysAcceleration()
   n3 = Wings_Movement_Poses_PoseAsset : Jump(PoseWeight=1) {poseAsset=&Wings_Movement_Poses_PoseAsset, poseName=Jump}
   Output Animation Pose(Result=Apply Additive(
     Base=@n3
     Additive=Blendspace Player 'BS_Wings_Leans_Ground'(X=@n2, Y=@n1) {blendSpace=&BS_Wings_Leans_Ground}
    ) {alphaBoolBlend {blendOption=Linear}})
  ]
 }
 SkydiveMode {
  path=AnimGraph.AnimGraphNode_StateMachine_0.Wings_position.AnimStateNode_4.SkydiveMode
  nodes [
   AnimGraphNode_StateMachine_0 = SkyDivingState() {is=title AnimGraphNode_StateMachine}
   Output Animation Pose(Result=@AnimGraphNode_StateMachine_0)
  ]
 }
 SkyDivingState {
  entry=Parashuting
  path=AnimGraph.AnimGraphNode_StateMachine_0.Wings_position.AnimStateNode_4.SkydiveMode.AnimGraphNode_StateMachine_0.SkyDivingState
  ids=id
  nodes [
   AnimStateNode_0 = Parashuting() {is=title AnimStateNode}
   AnimStateNode_1 = Diving() {is=title AnimStateNode}
   AnimStateNode_2 = floating() {is=title AnimStateNode}
   AnimStateTransitionNode_0 = AnimStateTransitionNode()
   AnimStateTransitionNode_1 = AnimStateTransitionNode()
   AnimStateTransitionNode_3 = AnimStateTransitionNode()
   AnimStateTransitionNode_4 = AnimStateTransitionNode()
   AnimStateTransitionNode_5 = AnimStateTransitionNode()
   AnimStateTransitionNode_6 = AnimStateTransitionNode()
  ]
  links [
   {
    dst=Parashuting
    rule=AnimGraph.AnimGraphNode_StateMachine_0.Wings_position.AnimStateNode_4.SkydiveMode.AnimGraphNode_StateMachine_0.SkyDivingState.AnimStateTransitionNode_1.Transition
    src=Diving
    via=AnimStateTransitionNode_1
   }
   {
    dst=floating
    rule=AnimGraph.AnimGraphNode_StateMachine_0.Wings_position.AnimStateNode_4.SkydiveMode.AnimGraphNode_StateMachine_0.SkyDivingState.AnimStateTransitionNode_3.Transition
    src=Diving
    via=AnimStateTransitionNode_3
   }
   {
    dst=Diving
    rule=AnimGraph.AnimGraphNode_StateMachine_0.Wings_position.AnimStateNode_4.SkydiveMode.AnimGraphNode_StateMachine_0.SkyDivingState.AnimStateTransitionNode_0.Transition
    src=Parashuting
    via=AnimStateTransitionNode_0
   }
   {
    dst=floating
    rule=AnimGraph.AnimGraphNode_StateMachine_0.Wings_position.AnimStateNode_4.SkydiveMode.AnimGraphNode_StateMachine_0.SkyDivingState.AnimStateTransitionNode_6.Transition
    src=Parashuting
    via=AnimStateTransitionNode_6
   }
   {
    dst=Diving
    rule=AnimGraph.AnimGraphNode_StateMachine_0.Wings_position.AnimStateNode_4.SkydiveMode.AnimGraphNode_StateMachine_0.SkyDivingState.AnimStateTransitionNode_4.Transition
    src=floating
    via=AnimStateTransitionNode_4
   }
   {
    dst=Parashuting
    rule=AnimGraph.AnimGraphNode_StateMachine_0.Wings_position.AnimStateNode_4.SkydiveMode.AnimGraphNode_StateMachine_0.SkyDivingState.AnimStateTransitionNode_5.Transition
    src=floating
    via=AnimStateTransitionNode_5
   }
  ]
 }
 Parashuting {
  path=AnimGraph.AnimGraphNode_StateMachine_0.Wings_position.AnimStateNode_4.SkydiveMode.AnimGraphNode_StateMachine_0.SkyDivingState.AnimStateNode_0.Parashuting
  nodes [
   n1 = Get PlayerForwardAcceleration()
   n2 = Get PlayerSidewaysAcceleration()
   n3 = Wings_Movement_Poses_PoseAsset : Float!(PoseWeight:float=1) -> (Pose:Pose Link Structure) {
    is=title AnimGraphNode_PoseByName
    poseAsset=&Wings_Movement_Poses_PoseAsset
    poseName=Float
   }
   Output Animation Pose(Result=Apply Additive(
     Base=@n3
     Additive=Blendspace Player 'BS_Wings_Parashuting'(X=@n2, Y=@n1) {blendSpace=/BRCosmetics/Accessories/FORT_Backpacks/Backpack_F_MED_Cavalry_Alt/Animation/BS_Wings_Parashuting}
    ) {alphaBoolBlend {blendOption=Linear}})
  ]
 }
 Diving {
  path=AnimGraph.AnimGraphNode_StateMachine_0.Wings_position.AnimStateNode_4.SkydiveMode.AnimGraphNode_StateMachine_0.SkyDivingState.AnimStateNode_1.Diving
  nodes [
   n1 = Get PlayerSidewaysAcceleration()
   n2 = Get PlayerForwardAcceleration()
   n3 = Wings_Movement_Poses_PoseAsset : Dive!(PoseWeight:float=1) -> (Pose:Pose Link Structure) {
    is=title AnimGraphNode_PoseByName
    poseAsset=&Wings_Movement_Poses_PoseAsset
    poseName=Dive
   }
   Output Animation Pose(Result=Apply Additive(
     Base=@n3
     Additive=Blendspace Player 'BS_Wings_Leans_Diving'(X=@n1, Y=@n2) {blendSpace=/BRCosmetics/Accessories/FORT_Backpacks/Backpack_F_MED_Cavalry_Alt/Animation/BS_Wings_Leans_Diving}
    ) {alphaBoolBlend {blendOption=Linear}})
  ]
 }
 floating {
  path=AnimGraph.AnimGraphNode_StateMachine_0.Wings_position.AnimStateNode_4.SkydiveMode.AnimGraphNode_StateMachine_0.SkyDivingState.AnimStateNode_2.floating
  nodes [
   n1 = Get PlayerForwardAcceleration()
   n2 = Get PlayerSidewaysAcceleration()
   n3 = Wings_Movement_Poses_PoseAsset : Float(PoseWeight=1) {poseAsset=&Wings_Movement_Poses_PoseAsset, poseName=Float}
   Output Animation Pose(Result=Apply Additive(
     Base=@n3
     Additive=Blendspace Player 'BS_Wings_Leans_Diving'(X=@n2, Y=@n1) {blendSpace=&BS_Wings_Leans_Diving}
    ) {alphaBoolBlend {blendOption=Linear}})
  ]
 }
 Transition {
  path=AnimGraph.AnimGraphNode_StateMachine_0.Wings_position.AnimStateNode_4.SkydiveMode.AnimGraphNode_StateMachine_0.SkyDivingState.AnimStateTransitionNode_0.Transition
  nodes [Result(bCanEnterTransition=NOT Boolean(A=Get bIsParachuteOpened()))]
 }
 Transition {
  path=AnimGraph.AnimGraphNode_StateMachine_0.Wings_position.AnimStateNode_4.SkydiveMode.AnimGraphNode_StateMachine_0.SkyDivingState.AnimStateTransitionNode_1.Transition
  nodes [Result(bCanEnterTransition=Get bIsParachuteOpened())]
 }
 Transition {
  path=AnimGraph.AnimGraphNode_StateMachine_0.Wings_position.AnimStateNode_4.SkydiveMode.AnimGraphNode_StateMachine_0.SkyDivingState.AnimStateTransitionNode_3.Transition
  nodes [Result(bCanEnterTransition=NOT Boolean(A=Get bIsSkydiveDiveMode()))]
 }
 Transition {
  path=AnimGraph.AnimGraphNode_StateMachine_0.Wings_position.AnimStateNode_4.SkydiveMode.AnimGraphNode_StateMachine_0.SkyDivingState.AnimStateTransitionNode_4.Transition
  nodes [Result(bCanEnterTransition=Get bIsSkydiveDiveMode())]
 }
 Transition {
  path=AnimGraph.AnimGraphNode_StateMachine_0.Wings_position.AnimStateNode_4.SkydiveMode.AnimGraphNode_StateMachine_0.SkyDivingState.AnimStateTransitionNode_5.Transition
  nodes [Result(bCanEnterTransition=Get bIsParachuteOpened())]
 }
 Transition {
  path=AnimGraph.AnimGraphNode_StateMachine_0.Wings_position.AnimStateNode_4.SkydiveMode.AnimGraphNode_StateMachine_0.SkyDivingState.AnimStateTransitionNode_6.Transition
  nodes [Result(bCanEnterTransition=NOT Boolean(A=Get bIsParachuteOpened()))]
 }
 Crouching {
  path=AnimGraph.AnimGraphNode_StateMachine_0.Wings_position.AnimStateNode_5.Crouching
  nodes [
   n1 = Wings_Movement_Poses_PoseAsset : Crouch!(PoseWeight:float=1) -> (Pose:Pose Link Structure) {
    is=title AnimGraphNode_PoseByName
    poseAsset=&Wings_Movement_Poses_PoseAsset
    poseName=Crouch
   }
   Output Animation Pose(Result=@n1)
  ]
 }
 Transition {
  path=AnimGraph.AnimGraphNode_StateMachine_0.Wings_position.AnimStateTransitionNode_15.Transition
  nodes [Result(bCanEnterTransition=Get bIsFalling())]
 }
 Transition {
  path=AnimGraph.AnimGraphNode_StateMachine_0.Wings_position.AnimStateTransitionNode_16.Transition
  nodes [Result(bCanEnterTransition=NOT Boolean(A=Get bIsFalling()))]
 }
 Transition {
  path=AnimGraph.AnimGraphNode_StateMachine_0.Wings_position.AnimStateTransitionNode_17.Transition
  nodes [Result(bCanEnterTransition=Get bIsFalling())]
 }
 Transition {
  path=AnimGraph.AnimGraphNode_StateMachine_0.Wings_position.AnimStateTransitionNode_18.Transition
  nodes [Result(bCanEnterTransition=Get bIsSkyDiving())]
 }
 Transition {
  path=AnimGraph.AnimGraphNode_StateMachine_0.Wings_position.AnimStateTransitionNode_19.Transition
  nodes [Result(bCanEnterTransition=NOT Boolean(A=Get bIsSkyDiving()))]
 }
 Transition {
  path=AnimGraph.AnimGraphNode_StateMachine_0.Wings_position.AnimStateTransitionNode_2.Transition
  nodes [Result(bCanEnterTransition=Get bIsJumping())]
 }
 Transition {
  path=AnimGraph.AnimGraphNode_StateMachine_0.Wings_position.AnimStateTransitionNode_20.Transition
  nodes [Result(bCanEnterTransition=Get bIsSkyDiving())]
 }
 Transition {
  path=AnimGraph.AnimGraphNode_StateMachine_0.Wings_position.AnimStateTransitionNode_21.Transition
  nodes [Result(bCanEnterTransition=Get bIsCrouching())]
 }
 Transition {
  path=AnimGraph.AnimGraphNode_StateMachine_0.Wings_position.AnimStateTransitionNode_22.Transition
  nodes [Result(bCanEnterTransition=NOT Boolean(A=Get bIsCrouching()))]
 }
 Transition {
  path=AnimGraph.AnimGraphNode_StateMachine_0.Wings_position.AnimStateTransitionNode_23.Transition
  nodes [Result(bCanEnterTransition=Get bIsTargeting())]
 }
 Transition {
  path=AnimGraph.AnimGraphNode_StateMachine_0.Wings_position.AnimStateTransitionNode_24.Transition
  nodes [Result(bCanEnterTransition=NOT Boolean(A=Get bIsTargeting()))]
 }
 Transition {
  path=AnimGraph.AnimGraphNode_StateMachine_0.Wings_position.AnimStateTransitionNode_3.Transition
  nodes [Result(bCanEnterTransition=NOT Boolean(A=Get bIsJumping()))]
 }
 EventGraph {nodes [
   TryGetPawnOwner()
   Event BlueprintUpdateAnimation()
  ]}
]
variables []
